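; Base Blueprint for the Figment Martin rock pickup, which players capture through a long-use interaction with audio feedback.
; The server counts interacting pawns, setting bIsInterracting when interaction starts and clearing it on completion or when the count empties.
; The interaction sound plays and fades with that flag, mirrored on clients by its OnRep.
; On destruction, a rock-to-backpack sound plays at the touch capsule if the interaction sound is not playing.
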(event EventBeginPlay
  (Game|AssignOnDestroyed (delegate Custom|OnDestroyed_Event_0)))

(event Collision|EventActorBeginOverlap
  (params (OtherActor :type Actor%20Object%20Reference)))

(event Custom|OnDestroyed_Event_0
  (params (DestroyedActor :type Actor%20Object%20Reference))
  (if (not (Audio|Components|Audio|IsPlaying (Variables|Default|GetInteractionSound)))
    (Audio|PlaySoundatLocation "/Game/Sounds/Fort_GamePlay_Sounds/Carmine/SC_Carmine_RockToBackpack_Cue.SC_Carmine_RockToBackpack_Cue" (Transformation|GetWorldLocation (Variables|Collision|GetTouchCapsule)))
    (Development|PrintString "THIS IS THE BOTTOM" true true "(R=0.000000,G=0.660000,B=1.000000,A=1.000000)" 1e+01)))

(event EventTick
  (params (DeltaSeconds :type Float%20%28single-precision%29)))

(event EventOnFillContextInfo
  (params (ContextInfo :type Fort%20Interact%20Context%20Info%20Object%20Reference))
  (Variables|/Script/FortniteGame.FortInteractContextInfo|SetbReplicateLongInteractionDetails true ContextInfo))

(event EventOnServerNotifyStartLongUse
  (params (InteractingPawn :type Fort%20Player%20Pawn%20Object%20Reference))
  ; When a player interacts with the pickup on the Server, keep track of how many players are interacting with the pickup and check if we should set IsInteracting to True.
  (Development|PrintString "OnServerNotifyStartLongUse Received" false)
  (bind _result (Math|Integer|IncrementInt (Variables|Default|GetNumInteractingPawns))
    (:%20%20
      (Variables|Default|SetNumInteractingPawns _result)
      (if (not (Variables|Default|GetbIsInterracting))
        (Variables|Default|SetbIsInterracting true)
        (Audio|Components|Audio|Play (Variables|Default|GetInteractionSound) 0.0)))))

(event EventOnServerNotifyEndLongUse
  (params (InteractingPawn :type Fort%20Player%20Pawn%20Object%20Reference) (bUseCompleted :type Boolean))
  ; When a player stops interacting with the pick on the Server,check if it was fully captured and if so set IsInteracting to False otherwise check if we still have players trying to capture the pickup.
  (Development|PrintString "OnServerNotifyEndLongUse Received" false)
  (if (not (and bUseCompleted (Variables|Default|GetbIsInterracting)))
    (bind _result (Math|Integer|DecrementInt (Variables|Default|GetNumInteractingPawns))
      (:%20%20
        (bind _output_get (Variables|Default|SetNumInteractingPawns _result))
        (if (and (<= _output_get 0) (Variables|Default|GetbIsInterracting))
          _))))
  (Variables|Default|SetbIsInterracting false)
  (Audio|Components|Audio|FadeOut (Variables|Default|GetInteractionSound) 0.2 0.0 "Linear" :!class /Script/Engine.AudioComponent))

(fn OnRep_bIsInterracting
  ; Play sound for clients when a player is interracting with the pickup and stop the sound when all players have stopped interracting with the pickup.
  (Networking|CanExecuteCosmeticEvents
    (:True
      (if (Variables|Default|GetbIsInterracting)
        (Audio|Components|Audio|Play (Variables|Default|GetInteractionSound) 0.0)
        (else
          (Audio|Components|Audio|FadeOut (Variables|Default|GetInteractionSound) 0.2 0.0 "Linear" :!class /Script/Engine.AudioComponent))))))

(fn ConstructionScript)
