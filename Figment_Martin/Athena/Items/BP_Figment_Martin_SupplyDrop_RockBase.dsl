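; Meteor-style supply drop base that falls, lands on ground or water, and spawns a rock artifact pickup with a bound marker actor.
; Building hits destroy the building and resume the fall; other hits land it, while water overlap opens it immediately.
; Server spawns loot once from ItemDefToSpawn.
; Clients strip flare and reticle components, delay crate collision with a timed ground trace, and use a landing-location RepNotify for delayed position correction.
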(event EventBeginPlay
  ;[1 1 1] BeginPlay
  (Utilities|FlowControl|Sequence
    (:then_0
      (Networking|SwitchHasAuthority
        (:Remote
          (Collision|SetCollisionEnabled (Variables|Default|GetCrateCollision) "NoCollision")
          (Components|DestroyComponent (Variables|Default|GetFlareAudio))
          (Components|DestroyComponent (Variables|Default|GetReticle_Niagara))
          (Components|DestroyComponent (Variables|Default|GetReticleAudio))
          (Components|DestroyComponent (Variables|Default|GetFlare_Niagara))
          (Utilities|Time|SetTimerbyFunctionName 0 "CheckGroundForCollisionToggle" 1.0 true)
          (Audio|SpawnSoundAttached "/Figment_Martin/Sounds/Fort_GamePlay_Sounds/Meteor/Fort_Meteor_Carmine_Appear_Athena_Cue.Fort_Meteor_Carmine_Appear_Athena_Cue" (Variables|Default|GetFallingAudio)))))
    (:then_1
      (|CallParentFunction)
      (Variables|Default|SetLootTableName "Loot_AthenaMeteorDrop"))))

(event EventHandleProjectileMovementStop
  (params (ImpactResult :type Hit%20Result%20Structure))
  (bind _returnvalue (Transformation|GetActorLocation))
  (bind (_bblockinghit _binitialoverlap _time _distance _location _impactpoint _normal _impactnormal _physmat _hitactor) (Collision|BreakHitResult ImpactResult))
  ;[1 1 1] HandleProjectileMovementStop
  (bind _asbuilding_actor (Utilities|Casting|CastToBuildingActor _hitactor)
    (:then
      ;;[0 0 0] If we hit a building, destroy it and keep falling
      (Actor|DestroyActor _asbuilding_actor)
      (Utilities|Time|SetTimerbyEvent (EventDispatchers|CreateEvent :!event Multicast_ApplyGravityForFall) 1e-05)
      (Utilities|FlowControl|DoOnce false
        (:Completed
          (Audio|SpawnSoundatLocation "/Figment_Martin/Sounds/Fort_GamePlay_Sounds/Meteor/Meteor_SubExplode_Athena_Cue.Meteor_SubExplode_Athena_Cue" _returnvalue)
          (Effects|SpawnEmitteratLocation "/Figment_Martin/Effects/Fort_Effects/LTM/MeteorSupplyDrop/P_Meteor_SupplyDrop_Impact.P_Meteor_SupplyDrop_Impact" _returnvalue)
          (Audio|Components|Audio|FadeOut (Variables|Default|GetFallingAudio) 0.2 0.0 "Linear" :!class /Script/Engine.AudioComponent))))
    (:CastFailed
      (CallFunction|Landed)
      (|CallParentFunction))))

(event Custom|Landed
  ;[1 1 1] Landed
  (Effects|Particles|Trails|EndTrails (Variables|Default|GetFireTrail))
  (Audio|SpawnSoundatLocation "/Figment_Martin/Sounds/Fort_GamePlay_Sounds/Meteor/Meteor_SubExplode_Athena_Cue.Meteor_SubExplode_Athena_Cue" (Transformation|GetActorLocation))
  (Audio|Components|Audio|FadeOut (Variables|Default|GetFallingAudio) 0.2 0.0 "Linear" :!class /Script/Engine.AudioComponent))

(event EventSpawnLoot
  (bind _lootspawnlocation (CallFunction|GetLootSpawnLocation))
  (bind _itemdeftospawn (Variables|Default|GetItemDefToSpawn))
  (Networking|SwitchHasAuthority
    (:Authority
      (Utilities|FlowControl|DoOnce false
        (:Completed
          (Utilities|FlowControl|Sequence
            (:then_0
              (if (Utilities|IsValid _itemdeftospawn :!class /Script/Engine.KismetSystemLibrary)
                (Audio|PlaySoundatLocation "/Figment_Martin/Sounds/Fort_GamePlay_Sounds/Meteor/Meteor_Carmine_Explode_Athena_Cue.Meteor_Carmine_Explode_Athena_Cue" _lootspawnlocation)
                (bind _returnvalue (CallFunction|SpawnGameModePickup _itemdeftospawn "/Figment_Martin/Athena/Items/BP_Figment_Martin_RockPickupBase.BP_Figment_Martin_RockPickupBase_C" 1 (Variables|Default|GetMyInteractingPawn) _lootspawnlocation))
                ;[1 1 1] Handle spawning a marker actor for the artifact
                (if (CallFunction|ShouldArtifactbeMarkedonLanded)
                  (bind _returnvalue_2 (Game|SpawnActorfromClass (Variables|Default|GetMarkerActorToSpawn) (Math|Transform|MakeTransform (Transformation|GetActorLocation _returnvalue)) "AlwaysSpawn" "MultiplyWithRoot" 0 0))
                  (bind _asmartin_rock_marker_base (Utilities|Casting|CastToMartin_RockMarkerBase _returnvalue_2))
                  (Class|MartinRockMarkerBase|BindtoArtifact _asmartin_rock_marker_base _returnvalue))))
            (:then_1
              (CallFunction|OnLootSpawned))))))))

(event Collision|EventActorBeginOverlap
  (params (OtherActor :type Actor%20Object%20Reference))
  ;[1 1 1] Handle landing on water
  (Utilities|Casting|CastToFortWaterBodyBP OtherActor)
  (CallFunction|InfinityStoneProjectileStop))

(fn CheckGroundForCollisionToggle
  (if (Utilities|Enum|Equal(Enum) (Collision|GetCollisionEnabled (Variables|Default|GetCrateCollision)) "NoCollision")
    (Collision|LineTraceByChannel (Transformation|GetWorldLocation (Variables|Default|GetCrateCollision)) (+ (Transformation|GetWorldLocation (Variables|Default|GetCrateCollision)) (* (Variables|Default|GetFallDirection) 8e+03)) "TraceTypeQuery10")
    (Collision|SetCollisionEnabled (Variables|Default|GetCrateCollision) "QueryAndPhysics")
    (Utilities|Time|ClearTimerbyFunctionName 0 "CheckGroundForCollisionToggle")))

(fn InfinityStoneProjectileStop
  (params (Hit%20Result :type Hit%20Result%20Structure))
  ;[1 1 1] Landed in water, open crate
  (Utilities|FlowControl|Sequence
    (:then_0
      (Utilities|Time|ClearTimerbyFunctionName 0 "Multicast_ApplyGravityForFall"))
    (:then_1
      (Networking|SwitchHasAuthority
        (:Authority
          (Variables|Default|SetLandingLocation1 (Transformation|GetActorLocation)))))
    (:then_2
      (CallFunction|PlayLandingFX true))
    (:then_3
      (CallFunction|OpenSupplyDrop))))

(fn OnRep_LandingLocation1
  (Networking|SwitchHasAuthority
    (:Remote
      (Variables|/Script/Engine.ProjectileMovementComponent|SetbSweepCollision true (Variables|Default|GetFortProjectileMovement))
      ;[1 1 1] Force client supply drop to server location after a slight delay if we haven't stopped due to parts of level not being streamed in
      (if (Math|Vector|Equal(Vector) (Variables|Default|GetLandingLocation) "0, 0, 0" 0.1)
        (Utilities|Time|ClearTimerbyFunctionName 0 "ForceSetLocationOnLanding")
        (else
          (Utilities|Time|SetTimerbyFunctionName 0 "ForceSetLocationOnLanding" 0.5))))))

(fn ShouldArtifactBeMarkedOnLanded
  (return true))

(fn ConstructionScript
  (|CallParentFunction))
